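; Actor component that drives band emote music on a player pawn as bandmates join and leave.
; Join/leave events update per-member volume targets and a bandmate count, whose OnRep handlers refresh audio.
; Emote audio components receive volume parameters, a FollowerCount lerped by a restartable next-tick loop, and a Group, 1P or 3P music category.
; Tick also calls UpdateComponent, which reapplies parameters and category.
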(event EventBeginPlay
  (Variables|Default|SetPlayerPawn (Default|GetFortPlayerPawn)))

(event EventTick
  (params (DeltaSeconds :type Float%20%28single-precision%29))
  (CallFunction|UpdateComponent))

(event Custom|Update%20Component
  (Default|UpdateSoundParams)
  (Default|EmoteAudioUpdateCategory))

(event Custom|BandMate%20Joined
  (params (BandMate%20Index :type Integer))
  (CallFunction|SetTargetVolume BandMate%20Index 1.0)
  (Default|BandMateCountIncrement))

(event Custom|BandMate%20Left
  (params (BandMate%20Index :type Integer))
  (CallFunction|SetTargetVolume BandMate%20Index)
  (Default|BandMateCountDecrement))

(event Custom|Update%20SoundPriority
  (Default|EmoteAudioUpdateCategory))

(event Custom|Interpolate%20Sound%20Params
  (Utilities|DoneInterpolating
    (:True
      (Variables|Default|SetInterpolation 0.0)
      (Utilities|FlowControl|DelayUntilNextTick)
      (Default|InterpolatedTimeUpdate)
      (Default|InterpolatedFollowerCountUpdate)
      (Utilities|DoneInterpolating
        (:False)))
    (:False
      (Variables|Default|SetInterpolation 0.0) ; Reset interpolation w/o creating a 2nd loop.
    )))

(fn BandMateCountDecrement
  (Variables|Default|SetBandMate%20Count (- (Variables|Default|GetBandMate%20Count) 1)))

(fn BandMateCountIncrement
  (Variables|Default|SetBandMate%20Count (+ (Variables|Default|GetBandMate%20Count) 1)))

(fn EmoteAudioSetCategory
  (params (Category :type Gameplay%20Tag%20Structure))
  (bind _playerpawn (Variables|Default|GetPlayerPawn))
  (Utilities|IsValid _playerpawn
    (:Is%20Valid
      (for _array_element (Emote|GetAllEmoteAudioComponents _playerpawn)
        (Audio|ChangeEventCategory (GameInstanceSubsystems|GetMusicEventSubsystem) _array_element Category)))))

(fn EmoteAudioSetFloat
  (params (Name :type Name) (Value :type Float%20%28double-precision%29))
  (bind _playerpawn (Variables|Default|GetPlayerPawn))
  (Utilities|IsValid _playerpawn
    (:Is%20Valid
      (for _array_element (Emote|GetAllEmoteAudioComponents _playerpawn)
        (Audio|Parameter|SetFloatParameter _array_element Name Value :!class /Script/AudioExtensions.AudioParameterControllerInterface)))))

(fn EmoteAudioUpdateCategory
  (if (> (Variables|Default|GetBandMate%20Count) 0)
    (CallFunction|EmoteAudioSetCategory "(TagName=\"MusicEventSubsystem.Event.Emote.Group\")")
    (elif (Pawn|IsLocallyControlled (Variables|Default|GetPlayerPawn))
      (CallFunction|EmoteAudioSetCategory "(TagName=\"MusicEventSubsystem.Event.Emote.1P\")")
      (else
        (CallFunction|EmoteAudioSetCategory "(TagName=\"MusicEventSubsystem.Event.Emote.3P\")")))))

(fn GetFortPlayerPawn
  (bind _asfort_player_pawn (Utilities|Casting|CastToFortPlayerPawn (Components|GetOwner))
    (:then
      (return _asfort_player_pawn))
    (:CastFailed
      (return 0))))

(fn InterpolatedFollowerCountInitialize
  (Variables|Default|SetFollowerCount%20From (Variables|Default|GetFollowerCount))
  (Variables|Default|SetFollowerCount%20To (Math|Conversions|ToFloat(Integer) (Variables|Default|GetBandMate%20Count))))

(fn InterpolatedFollowerCountUpdate
  (bind _alpha (Default|InterpolatedTimeAlphaGet))
  (bind _output_get (Variables|Default|SetFollowerCount (Math|Float|Lerp (Variables|Default|GetFollowerCount%20From) (Variables|Default|GetFollowerCount%20To) _alpha)))
  (Default|EmoteAudioSetFloat "FollowerCount" _output_get))

(fn InterpolatedTimeUpdate
  (Variables|Default|SetInterpolation (+ (Variables|Default|GetInterpolation) (Utilities|Time|GetWorldDeltaSeconds))))

(fn InterpolatedTimeAlphaGet
  (return (Math|Float|Clamp(Float) (Math|Float|SafeDivide (Variables|Default|GetInterpolation) (Variables|Default|GetInterpolation%20Time)))))

(fn OnRep_BandMateCount
  (Default|EmoteAudioUpdateCategory)
  (Default|InterpolatedFollowerCountInitialize)
  (CallFunction|InterpolateSoundParams))

(fn OnRep_VolumeTarget
  (Default|UpdateSoundParams))

(fn SetTargetVolume
  (params (Index :type Integer) (Volume :type Float%20%28double-precision%29))
  (bind _volume_target (Variables|Default|GetVolume%20Target))
  (if (Utilities|Array|IsValidIndex _volume_target Index)
    (Utilities|Array|SetArrayElem _volume_target Index Volume)))

(fn UpdateSoundParams
  (Default|EmoteAudioSetFloat "FollowerCount" (Variables|Default|GetFollowerCount))
  (for (_array_element _array_index) (Variables|Default|GetVolume%20Param)
    (Default|EmoteAudioSetFloat _array_element (Utilities|Array|Get(acopy) (Variables|Default|GetVolume%20Target) _array_index))))
