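; Dance emote gameplay ability for group emotes, letting other players join the player pawn's dance using the same current emote.
; It overrides a parent band ability, setting each joining mate's tail target to the player pawn and appending mates even when the parent add fails.
; The player pawn is always interactable; when followers may interact, valid band mates are flagged too and their interaction events are bound to the parent's handler.
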(event EventOnBandMateJoined
  (params (BandMate :type Fort%20Player%20Pawn%20Object%20Reference))
  (Variables|/Script/FortniteGame.FortPlayerPawn|SetGroupEmoteTailTarget (Variables|Default|GetPlayerPawn) BandMate)
  (|CallParentFunction BandMate))

(event EventUpdateInteractable
  ;[1 1 1] Update whether the emote is interactable. (Dance emotes always are)
  (bind _output_get (Variables|/Script/FortniteGame.FortPlayerPawn|SetbIsWaitingForEmoteInteraction true (Variables|Default|GetPlayerPawn)))
  (if (Variables|Default|GetbAllowFollowersToInteract)
    (for _array_element (Variables|Default|GetBandMates)
      (Utilities|IsValid _array_element
        (:Is%20Valid
          (bind _output_get_2 (Variables|/Script/FortniteGame.FortPlayerPawn|SetbIsWaitingForEmoteInteraction _output_get _array_element))
          (if _output_get_2
            (Game|Player|BindEventtoOnEmoteInteracted _array_element (delegate EventOnInteracted))
            (else
              (Game|Player|UnbindEventfromOnEmoteInteracted _array_element (delegate EventOnInteracted)))))))))

(event EventOnInteracted
  (params (Instigator :type Fort%20Player%20Pawn%20Object%20Reference))
  (|CallParentFunction Instigator))

(fn BandMateAdd
  (params (BandMate :type Fort%20Player%20Pawn%20Object%20Reference))
  (bind (_succeeded _index) (|CallParentFunction BandMate))
  (if _succeeded
    (return true _index)
    (else
      (bind _returnvalue (Utilities|Array|Add (Variables|Default|GetBandMates) BandMate))
      (return true _returnvalue))))

(fn GetFollowerEmote
  (params (Index :type Integer))
  (return (CallFunction|GetCurrentEmoteItemDefinition)))
