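; Gameplay ability for a band group emote, where a leader starts playing and interacting players join as bandmates with their own slot's emote.
; BandMates holds one slot per FollowerEmotes entry, filled or cleared on interact and leave events, which update interactability and a server-added AC_EmoteBand component.
; An optional random leader slot is stored on the pawn; clients may delay one tick for it.
; Ending notifies bandmates and removes the component.
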(event Custom|Update%20Interactable
  ;[1 1 1] Update whether the emote is interactable.
  (bind _output_get (Variables|/Script/FortniteGame.FortPlayerPawn|SetbIsWaitingForEmoteInteraction (> (Utilities|Array|FindItem (Variables|Default|GetBandMates)) -1) (Variables|Default|GetPlayerPawn)))
  (if (Variables|Default|GetbAllowFollowersToInteract)
    (for _array_element (Variables|Default|GetBandMates)
      (Utilities|IsValid _array_element
        (:Is%20Valid
          (bind _output_get_2 (Variables|/Script/FortniteGame.FortPlayerPawn|SetbIsWaitingForEmoteInteraction _output_get _array_element))
          (if _output_get_2
            (Game|Player|BindEventtoOnEmoteInteracted _array_element (delegate Custom|OnInteracted))
            (else
              (Game|Player|UnbindEventfromOnEmoteInteracted _array_element (delegate Custom|OnInteracted)))))))))

(event Custom|On%20BandMate%20Joined
  (params (BandMate :type Fort%20Player%20Pawn%20Object%20Reference))
  ;[1 1 1] On BandMate Joined
  (bind (_succeeded _index) (CallFunction|BandMateAdd BandMate))
  (if _succeeded
    (Class|ACEmoteBand|BandMateJoined (Variables|Default|GetEmoteSyncComponent) _index)
    (Class|FortPlayerPawn|PlayGroupEmote BandMate (CallFunction|GetFollowerEmote _index))
    (Variables|/Script/FortniteGame.FortPlayerPawn|SetGroupEmoteTailTarget (Variables|Default|GetPlayerPawn) BandMate)))

(event Custom|On%20BandMate%20Left
  (params (BandMate :type Fort%20Player%20Pawn%20Object%20Reference))
  ;[1 1 1] On BandMate Left
  (bind (_succeeded _index) (Default|BandMateRemove BandMate))
  (if _succeeded
    (Class|ACEmoteBand|BandMateLeft (Variables|Default|GetEmoteSyncComponent) _index)
    (CallFunction|UpdateInteractable)
    (if (Variables|Default|GetbAllowFollowersToInteract)
      (Utilities|IsValid (Variables|/Script/FortniteGame.FortPawn|GetLastEmoteItemDef BandMate)
        (:Is%20Valid
          _)
        (:Is%20Not%20Valid
          (Variables|/Script/FortniteGame.FortPlayerPawn|SetbIsWaitingForEmoteInteraction false BandMate)))
      (Game|Player|UnbindEventfromOnEmoteInteracted BandMate (delegate Custom|OnInteracted)))))

(event Ability|EventActivateAbility
  (Utilities|IsServer
    (:true
      (CallFunction|AssignLeaderEmoteIndex)
      (|CallParentFunction))
    (:false
      (CallFunction|WaitForReplicatedValues))))

(event EventPlayInitialEmoteMontage
  (bind _montageinfo (Variables|Animation|GetMontageInfo))
  (bind _montagetoplay (Utilities|Struct|BreakFortGameplayAbilityMontageInfo _montageinfo))
  (bind _returnvalue (CallFunction|GetCurrentEmoteItemDefinition))
  (bind _playerpawn (Variables|Default|GetPlayerPawn))
  ;[1 1 1] Activate the montage
  (Utilities|IsValid _returnvalue
    (:Is%20Valid
      (if (Emote|ShouldHolsterWeapon _returnvalue _playerpawn)
        (Variables|/Script/FortniteGame.FortPlayerPawn|SetbIsWaitingForEmoteInteraction true _playerpawn)
        (Weapon|SetHolsterWeapon (Variables|Default|GetPlayerPawn) 0 true false (Variables|Default|GetEmoteHolsterId))
        (else
          _))
      (if (not (Variables|Default|GetbAbilityStopped))
        (CallFunction|InitializeAudio)
        (CallFunction|BindToEvents)
        (Ability|Tasks|PlayMontageInfoAndWaitForTrigger "None" _montageinfo
          (:then
            (if (Variables|Default|GetbAbilityStopped)
              (CallFunction|ForceStopMontage)
              (else
                (Utilities|FlowControl|Sequence
                  (:then_0
                    ;[1 1 1] Show/hide reticle if requested
                    (if (Variables|Reticle|GetHideReticle)
                      (CallFunction|SetReticleVisibility true)))))))
          (:Cancelled
            _)
          (:Completed
            _))
        (Utilities|FlowControl|Sequence
          (:then_0
            (Weapon|SetHolsterWeapon (Variables|Default|GetPlayerPawn) 0 false false (Variables|Default|GetEmoteHolsterId)))
          (:then_1
            (Utilities|IsValid _montagetoplay
              (:Is%20Valid
                (Utilities|FlowControl|Delay (Montage|GetDefaultBlendOutTime _montagetoplay)))
              (:Is%20Not%20Valid
                _))
            (Ability|EndAbility)))))
    (:Is%20Not%20Valid
      _)))

(event Custom|Initialize%20Audio
  ;[1 1 1] Initialize Audio
  (CallFunction|BandMatesInitialize)
  (Utilities|IsServer
    (:true
      (bind _returnvalue (Components|AddDynamicComponentbyClass (Variables|Default|GetPlayerPawn) "/BRCosmetics/Abilities/Emotes/AC_EmoteBand.AC_EmoteBand_C"))
      (bind _asac_emote_band (Utilities|Casting|CastToAC_EmoteBand _returnvalue))
      (Variables|Default|SetEmoteSyncComponent _asac_emote_band))))

(event Custom|Bind%20to%20Events
  (bind _playerpawn (Variables|Default|GetPlayerPawn))
  ;[1 1 1] Bind to Events
  (Game|Player|BindEventtoOnEmoteInteracted _playerpawn (delegate Custom|OnInteracted))
  (EventDispatchers|AssignOnGroupEmoteFollowerLeft _playerpawn (delegate Custom|OnFollowerLeft)))

(event Custom|OnInteracted
  (params (Instigator :type Fort%20Player%20Pawn%20Object%20Reference))
  ;[1 1 1] On Interacted
  (CallFunction|OnBandMateJoined Instigator)
  (CallFunction|UpdateInteractable))

(event Custom|OnFollowerLeft
  (params (LeavingPawn :type Fort%20Player%20Pawn%20Object%20Reference))
  ;[1 1 1] On Follower Left
  (CallFunction|OnBandMateLeft LeavingPawn))

(event Custom|Wait%20for%20Replicated%20Values
  ;[1 1 1] Wait for Replicated Values
  (if (Variables|Default|GetbRandomizeLeaderEmoteIndex)
    (bind _playerpawn (Default|GetPlayerPawn))
    (if (not (> (Variables|/Script/FortniteGame.FortPlayerPawn|GetEmoteBandLeaderIndex _playerpawn) -1))
      (Utilities|FlowControl|DelayUntilNextTick))
    (Variables|Default|SetLeaderEmoteIndex (Variables|/Script/FortniteGame.FortPlayerPawn|GetEmoteBandLeaderIndex _playerpawn))
    (|CallParentFunction)
    (else
      (|CallParentFunction))))

(fn AssignLeaderEmoteIndex
  (if (Variables|Default|GetbRandomizeLeaderEmoteIndex)
    (bind _playerpawn (Default|GetPlayerPawn))
    (if (< (Variables|/Script/FortniteGame.FortPlayerPawn|GetEmoteBandLeaderIndex _playerpawn) 0)
      (bind _output_get (Variables|Default|SetLeaderEmoteIndex (Math|Random|RandomInteger (Utilities|Array|Length (Variables|Default|GetFollowerEmotes)))))
      (Variables|/Script/FortniteGame.FortPlayerPawn|SetEmoteBandLeaderIndex _output_get _playerpawn)
      (else
        (Variables|Default|SetLeaderEmoteIndex (Variables|/Script/FortniteGame.FortPlayerPawn|GetEmoteBandLeaderIndex _playerpawn))))))

(fn BandMateAdd
  (params (BandMate :type Fort%20Player%20Pawn%20Object%20Reference))
  (bind _bandmates (Variables|Default|GetBandMates))
  (bind _bandmate_index (LocalVariables|GetBandMate%20Index))
  ;[1 1 1] If (due to reasons) the BandMate is already in the list, just return their previous index.
  ; See FORT-499114.
  (bind _output_get (LocalVariables|SetBandMate%20Index (Utilities|Array|FindItem _bandmates BandMate)))
  (if (> _output_get -1)
    (return true (LocalVariables|GetBandMate%20Index))
    (else
      ;[1 1 1] If an open slot exists, assign the new BandMate to that index.
      (bind _output_get_2 (LocalVariables|SetBandMate%20Index (Utilities|Array|FindItem _bandmates)))
      (if (== _output_get_2 -1)
        (return false -1)
        (else
          (Utilities|Array|SetArrayElem _bandmates _bandmate_index (Variables|GetBandMate))
          (return true _bandmate_index))))))

(fn BandMateRemove
  (params (BandMate :type Fort%20Player%20Pawn%20Object%20Reference))
  (bind _bandmates (Variables|Default|GetBandMates))
  (bind _output_get (LocalVariables|SetBandMate%20Index (Utilities|Array|FindItem _bandmates BandMate)))
  (if (> _output_get -1)
    (Utilities|Array|SetArrayElem _bandmates _output_get))
  (return (> _output_get -1) (LocalVariables|GetBandMate%20Index)))

(fn BandMatesCount
  (for _array_element (Variables|Default|GetBandMates)
    (Utilities|IsValid _array_element
      (:Is%20Valid
        (bind _result (Math|Integer|IncrementInt (LocalVariables|GetValid%20Bandmate%20Count))
          (:%20%20
            (LocalVariables|SetValid%20Bandmate%20Count _result))))))
  (return (LocalVariables|GetValid%20Bandmate%20Count)))

(fn BandMatesInitialize
  (bind _leaderemoteindex (Variables|Default|GetLeaderEmoteIndex))
  (Utilities|Array|Clear (Variables|Default|GetBandMates))
  ;[1 1 1] Add a bandmate slot for each follower emote specified.
  (for _ (Variables|Default|GetFollowerEmotes)
    (Utilities|Array|Add (Variables|Default|GetBandMates)))
  (if (> _leaderemoteindex -1)
    (Utilities|Array|SetArrayElem (Variables|Default|GetBandMates) _leaderemoteindex (Variables|Default|GetPlayerPawn))))

(fn BandMatesNotifyLeaderLeft
  (bind _playerpawn (Variables|Default|GetPlayerPawn))
  (for _array_element (Variables|Default|GetBandMates)
    (Utilities|IsValid _array_element
      (:Is%20Valid
        (if (not (== _array_element _playerpawn))
          (EventDispatchers|CallOnGroupEmoteFollowerLeft _array_element _playerpawn))))))

(fn GetPlayerPawn
  ;[1 1 1] Use this function instead of the 'PlayerPawn' variable if your code can run before 'ActivateAbility' is complete.
  (bind _asfort_player_pawn (Utilities|Casting|CastToFortPlayerPawn (Ability|GetActivatingPawn))
    (:then
      (return _asfort_player_pawn))
    (:CastFailed
      (return 0))))

(fn GetRandomFollowerEmote
  (return (CallFunction|GetFollowerEmote (Math|Random|RandomInteger (Utilities|Array|Length (Variables|Default|GetFollowerEmotes))))))

(fn GetFollowerEmote
  (params (Index :type Integer))
  (return (CallFunction|GetCurrentProductEmoteItemDefinition (Utilities|Array|Get(acopy) (Variables|Default|GetFollowerEmotes) Index))))

(fn GetMontageToPlay
  (params (EmoteAsset :type Fort%20Montage%20Item%20Definition%20Base%20Object%20Reference) (BodyType :type EFortCustomBodyType%20Enum :default "NONE") (Gender :type EFortCustomGender%20Enum :default "Invalid"))
  (if (> (Variables|Default|GetLeaderEmoteIndex) -1)
    (return (Emote|GetAnimation (CallFunction|GetFollowerEmote (Variables|Default|GetLeaderEmoteIndex)) BodyType Gender (Variables|Default|GetPlayerPawn)))
    (else
      (bind _returnvalue (|CallParentFunction EmoteAsset BodyType Gender))
      (return _returnvalue))))

(fn OnEndAbility
  (params (bWasCancelled :type Boolean))
  (bind _playerpawn (Variables|Default|GetPlayerPawn))
  (|CallParentFunction bWasCancelled)
  (Utilities|IsServer
    (:true
      (CallFunction|BandMatesNotifyLeaderLeft)
      (Components|DestroyDynamicComponent _playerpawn (Variables|Default|GetEmoteSyncComponent))
      (Variables|/Script/FortniteGame.FortPlayerPawn|SetEmoteBandLeaderIndex -1 _playerpawn))))
