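; Gameplay ability for a follower in the band group emote, stopping the montage when the follower strays or its tail target leaves.
; Non-server instances run a two-second timer comparing distance to the group emote tail target against a maximum.
; A follower-left dispatcher on the pawn triggers the stop; ending the ability fires that dispatcher on the target and clears the link.
; The sync-value handler is bound but empty.
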(event Custom|CheckDistance
  ;[1 1 1] Stop the Follower if the Leader has gotten too far away (check occurs every 2s on follower's client-only)
  (bind _distance (CallFunction|DistancetoLeader))
  (if (> _distance (Variables|Default|GetMaxDistanceFromLeader))
    (CallFunction|ForceStopMontage)))

(event Ability|EventActivateAbility
  ;[1 1 1] Active Ability
  (|CallParentFunction)
  (EventDispatchers|BindEventtoOnGroupEmoteFollowerLeft (Variables|Default|GetPlayerPawn) (delegate Custom|OnFollowerLeft))
  (Utilities|IsServer
    (:false
      (Utilities|Time|SetTimerbyEvent (delegate Custom|CheckDistance) 2.0 true))))

(event Custom|OnFollowerLeft
  (params (LeavingPawn :type Fort%20Player%20Pawn%20Object%20Reference))
  ;[1 1 1] On Follower Left
  (if (== LeavingPawn (Variables|/Script/FortniteGame.FortPlayerPawn|GetGroupEmoteTailTarget (Variables|Default|GetPlayerPawn)))
    (CallFunction|ForceStopMontage)))

(event Ability|EventOnEndAbility
  (params (bWasCancelled :type Boolean))
  (bind _playerpawn (Variables|Default|GetPlayerPawn))
  (bind _groupemotetailtarget (Variables|/Script/FortniteGame.FortPlayerPawn|GetGroupEmoteTailTarget _playerpawn))
  ;[1 1 1] Notify leader that this follower has left the emote.
  (|CallParentFunction bWasCancelled)
  (Utilities|IsValid _groupemotetailtarget
    (:Is%20Valid
      (EventDispatchers|CallOnGroupEmoteFollowerLeft _groupemotetailtarget _playerpawn))
    (:Is%20Not%20Valid
      _))
  (Variables|/Script/FortniteGame.FortPlayerPawn|SetGroupEmoteTailTarget 0 _playerpawn))

(event EventPlayInitialEmoteMontage
  (|CallParentFunction)
  (EventDispatchers|BindEventtoOnGroupEmoteSyncValueChanged (Variables|Default|GetPlayerPawn) (delegate Custom|Member%20Added)))

(event Custom|Member%20Added
  (params (NewSyncValue :type Byte)))

(fn DistanceToLeader
  (bind _playerpawn (Variables|Default|GetPlayerPawn))
  (bind _groupemotetailtarget (Variables|/Script/FortniteGame.FortPlayerPawn|GetGroupEmoteTailTarget _playerpawn))
  (Utilities|IsValid _playerpawn
    (:Is%20Valid
      (Utilities|IsValid _groupemotetailtarget
        (:Is%20Valid
          (return (Math|Vector|VectorLength (- (Transformation|GetActorLocation _groupemotetailtarget) (Transformation|GetActorLocation _playerpawn)))))))))
