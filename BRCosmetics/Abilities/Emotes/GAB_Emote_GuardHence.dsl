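; GAB_Emote_GuardHence is an emote gameplay ability that picks a band leader emote index from the character cosmetic the player has equipped.
; Its AssignLeaderEmoteIndex override calls the parent, then checks the Character loadout slot's metatags in order for EmberYearn, EagerUnify or AlderOptic.
; The first match sets LeaderEmoteIndex and copies it to the pawn's EmoteBandLeaderIndex.
; Activation and end events only hold disabled parent-call stubs.
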(event Ability|EventActivateAbility
  (|CallParentFunction) ; This node is disabled and will not be called.\nDrag off pins to build functionality.
)

(event Ability|EventOnEndAbility
  (params (bWasCancelled :type Boolean))
  (|CallParentFunction bWasCancelled) ; This node is disabled and will not be called.\nDrag off pins to build functionality.
)

(fn AssignLeaderEmoteIndex
  (bind _playerpawn (Default|GetPlayerPawn))
  (|CallParentFunction)
  (if (GameplayTags|HasTag (CosmeticLoadout|GatherItemDefinitionMetaTagsfromActor _playerpawn "(TagName=\"LoadoutSlot.Character\")") "(TagName=\"Cosmetics.Metatag.EmberYearn\")" true)
    (bind _output_get (Variables|Default|SetLeaderEmoteIndex 0))
    (Variables|/Script/FortniteGame.FortPlayerPawn|SetEmoteBandLeaderIndex _output_get _playerpawn)
    (elif (GameplayTags|HasTag (CosmeticLoadout|GatherItemDefinitionMetaTagsfromActor _playerpawn "(TagName=\"LoadoutSlot.Character\")") "(TagName=\"Cosmetics.Metatag.EagerUnify\")" true)
      (bind _output_get_2 (Variables|Default|SetLeaderEmoteIndex 1))
      (Variables|/Script/FortniteGame.FortPlayerPawn|SetEmoteBandLeaderIndex _output_get_2 _playerpawn)
      (elif (GameplayTags|HasTag (CosmeticLoadout|GatherItemDefinitionMetaTagsfromActor _playerpawn "(TagName=\"LoadoutSlot.Character\")") "(TagName=\"Cosmetics.Metatag.AlderOptic\")" true)
        (bind _output_get_3 (Variables|Default|SetLeaderEmoteIndex 2))
        (Variables|/Script/FortniteGame.FortPlayerPawn|SetEmoteBandLeaderIndex _output_get_3 _playerpawn)))))
