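; Gameplay ability for the HighTowerDate emote that swaps in a dedicated camera while the emote runs.
; On activation it calls the parent logic, then pushes the Athena_PlayerCameraMode_Date camera mode.
; On end, whether cancelled or not, it calls the parent and pops that camera mode.
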(event Ability|EventActivateAbility
  (|CallParentFunction)
  (Camera|PushCameraMode "/BRCosmetics/Blueprints/Camera/Athena/Athena_PlayerCameraMode_Date.Athena_PlayerCameraMode_Date_C"))

(event Ability|EventOnEndAbility
  (params (bWasCancelled :type Boolean))
  (|CallParentFunction bWasCancelled)
  (Camera|PopCameraMode))
