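; Emote ability for Hightower Mango that swaps the player's cosmetics to impersonate their last victim, falling back to a default outfit.
; A montage notify triggers a brief slow and an authority-only swap, using the data-store loadout or legacy tracked-victim lookup.
; If the mesh swap interrupts the montage, a no-blend version resumes at PostSwapJump.
; A newly applied swap also grants a cleanup ability that monitors the pawn.
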(event EventPlayInitialEmoteMontage
  (bind (_montagetoplay _animplayrate) (Utilities|Struct|BreakFortGameplayAbilityMontageInfo (Variables|Animation|GetMontageInfo)))
  (bind _returnvalue (CallFunction|GetCurrentEmoteItemDefinition))
  ;[1 1 1] basic setup from the GAB_Emote_Generic version of this
  (Utilities|Array|Clear (Variables|Default|GetActiveSwaps))
  (Utilities|IsValid _returnvalue
    (:Is%20Valid
      (if (Emote|ShouldHolsterWeapon _returnvalue (Variables|Default|GetPlayerPawn))
        (Weapon|SetHolsterWeapon (Variables|Default|GetPlayerPawn) 0 true false (Variables|Default|GetEmoteHolsterId))
        (else
          _))
      (if (not (Variables|Default|GetbAbilityStopped))
        ;[1 1 1] We need two versions of the montage so that we can skip the blend when we have to handle the mesh swap (gender or body type mismatch)
        (bind (_asynctaskproxy _notifytag) (Ability|Tasks|PlayMontageAndWaitNotify "GAB_Emote_HighTowerMango_Primary" _montagetoplay _animplayrate)
          (:then
            ;[1 1 1] We want to ensure that starting the montage doesn't cancel us for whatever reason
            (if (Variables|Default|GetbAbilityStopped)
              (CallFunction|ForceStopMontage)
              (elif (Variables|Reticle|GetHideReticle)
                ;;[1 1 1] Show/hide reticle if requested
                (CallFunction|SetReticleVisibility true))))
          (:OnCompleted
            (Variables|Default|SetRestartAtSectionWhenCancelled false))
          (:OnInterrupted
            _)
          (:OnCancelled
            _)
          (:OnNotifyBegin
            (switch Utilities|FlowControl|Switch|SwitchonGameplayTag _notifytag
              (:Ability.Trigger.Hightower.MangoCosmetic
                (CallFunction|PerformSwap)))))
        (CallFunction|CompletedorCancelled)))
    (:Is%20Not%20Valid
      _)))

(event Custom|PerformSwap
  ;[1 1 1] We want to disallow players from cancelling the emote during the critical window around the parts swapping...
  (CallFunction|ApplySlowtoPlayer)
  ;[1 1 1] If we're interrupted during this window we want to try to sync the animation up because this probably means that we're changing meshes here...
  (Variables|Default|SetRestartAtSectionWhenCancelled true)
  (CallFunction|StartSuitableSwap)
  (Utilities|FlowControl|Delay 0.6) ; We can play around with this timing
  (Variables|Default|SetRestartAtSectionWhenCancelled false)
  (CallFunction|RemoveSlowfromPlayer) ; Now we no longer care about the slow
)

(event Custom|Restart
  (bind (_montagetoplay _animplayrate) (Utilities|Struct|BreakFortGameplayAbilityMontageInfo (Variables|Animation|GetMontageInfo)))
  ;[1 1 1] We need two versions of the montage so that we can skip the blend when we have to handle the mesh swap (gender or body type mismatch)
  (if (Variables|Default|GetbAbilityStopped)
    (CallFunction|ForceStopMontage)
    (else
      (Ability|Tasks|PlayMontageAndWaitNotify "GAB_Emote_HighTowerMango_Secondary" (Variables|Default|GetNoBlendInMontage) _animplayrate "PostSwapJump"
        (:then
          (CallFunction|StopMontageifNeeded))
        (:OnCompleted
          _)
        (:OnInterrupted
          _)
        (:OnCancelled
          _))
      ;;[1 1 1] If this one is also interrupted we don't care to restart again
      (Variables|Default|SetRestartAtSectionWhenCancelled false)
      (CallFunction|CompletedorCancelled))))

(event Custom|CompletedOrCancelled
  (bind _montageinfo (Variables|Animation|GetMontageInfo))
  (bind _montagetoplay (Utilities|Struct|BreakFortGameplayAbilityMontageInfo _montageinfo))
  ;[1 1 1] This handles either cleanup or restarting
  (if (Variables|Default|GetRestartAtSectionWhenCancelled)
    (Utilities|FlowControl|Delay 0.0) ; we need time to have the mesh swap
    (Variables|Default|SetRestartAtSectionWhenCancelled false)
    (CallFunction|Restart)
    (else
      (Utilities|FlowControl|Sequence
        (:then_0
          (Weapon|SetHolsterWeapon (Variables|Default|GetPlayerPawn) 0 false false (Variables|Default|GetEmoteHolsterId)))
        (:then_1
          (Utilities|IsValid _montagetoplay
            (:Is%20Valid
              (Utilities|FlowControl|Delay (Montage|GetDefaultBlendOutTime _montagetoplay))
              (Ability|EndAbility))))))))

(event Ability|EventActivateAbility
  (|CallParentFunction)
  (Development|PrintString "Mango Activated With NO Tag"))

(event Ability|EventActivateAbilityFromEvent
  (params (EventData :type Gameplay%20Event%20Data%20Structure%20%28by%20ref%29))
  (Development|PrintString "Mango Activated By Tag"))

(fn ApplySlowToPlayer
  (if (not (Variables|Default|GetbHasAppliedSlow))
    (bind _returnvalue (Ability|ApplyGameplayEffectToOwner "/Game/Abilities/NPC/Generic/GE_PlayerDBNO_MovementSpeed.GE_PlayerDBNO_MovementSpeed_C"))
    (Variables|Default|SetAppliedSlow _returnvalue)
    (Ability|GameplayEffect|AssignActiveGameplayEffectTagSetbyCallerMagnitude _returnvalue "(TagName=\"SetByCaller.MoveSpeed\")")
    (Variables|Default|SetbHasAppliedSlow true)))

(fn ForceStopMontage
  (bind _returnvalue (Ability|GetActivatingPawn))
  (bind _mesh (Variables|/Script/Engine.Character|GetMesh _returnvalue))
  (bind _returnvalue_2 (Components|SkeletalMesh|GetAnimInstance _mesh))
  (bind _noblendinmontage (Variables|Default|GetNoBlendInMontage))
  (|CallParentFunction)
  (Utilities|IsValid _mesh
    (:Is%20Valid
      (Utilities|IsValid _returnvalue_2
        (:Is%20Valid
          (Utilities|IsValid _noblendinmontage
            (:Is%20Valid
              (Animation|Montage|MontageStop _returnvalue_2 (Montage|GetDefaultBlendOutTime _noblendinmontage) _noblendinmontage))))))))

(fn OnEndAbility
  (params (bWasCancelled :type Boolean))
  ;[1 1 1] We have to remove the slow no matter what!
  (CallFunction|RemoveSlowfromPlayer)
  (Utilities|FlowControl|Sequence
    (:then_1
      (|CallParentFunction bWasCancelled))))

(fn RemoveSlowFromPlayer
  (if (Variables|Default|GetbHasAppliedSlow)
    (Ability|RemoveGameplayEffectFromOwnerWithHandle (Variables|Default|GetAppliedSlow)) ; It's not ideal to call this with an invalid handle
    (Variables|Default|SetAppliedSlow)
    (Variables|Default|SetbHasAppliedSlow false)))

(fn StartSuitableSwap
  (bind _playerpawn (Variables|Default|GetPlayerPawn))
  (bind _playerpawn_2 (Variables|Default|GetPlayerPawn))
  (bind _swapguid (Variables|Default|GetSwapGUID))
  (bind _activeswaps (Variables|Default|GetActiveSwaps))
  (if (> (ScalableFloat|GetValueatLevel (Variables|Default|GetDatatableEnabled)) 0.0)
    (if (not (FortPawn|IsDead _playerpawn))
      (Ability|HasAuthority
        (:True
          (LocalVariables|SetbWasSwapActive (Class|FortPlayerPawn|IsItemandVariantSwapActivebyId _playerpawn (Variables|Default|GetSwapGUID)))
          (bind _asfort_player_controller_athena (Utilities|Casting|CastToFortPlayerControllerAthena (Pawn|GetController (Variables|Default|GetPlayerPawn)))
            (:then
              (bind _output_get (Variables|Default|SetDataStoreManager (Actor|GetComponentbyClass _asfort_player_controller_athena "/Script/FortniteGame.FortGenericDataStoreManagerComponent")))
              (Utilities|IsValid _output_get
                (:Is%20Valid
                  ;[1 1 1] Retrieve the victim's loadout from our data store - IDEAL PATH
                  (bind _output_get_2 (Utilities|GetMangoDataStore)
                    (:Is%20Valid
                      (if (and (> (Utilities|Array|Length (Variables|/Script/FortniteGame.FortDataStore_HightowerMangoDataStore|GetLastVictimRelevantLoadout _output_get_2)) 0) (> (ScalableFloat|GetValueatLevel (Variables|Default|GetDatatableImpersonationEnabled)) 0.0))
                        (Class|FortPlayerPawn|SwapPawnCosmeticItemandVariant (Variables|Default|GetPlayerPawn) (Variables|Default|GetSwapGUID) (Variables|/Script/FortniteGame.FortDataStore_HightowerMangoDataStore|GetLastVictimRelevantLoadout _output_get_2) false "None" false)
                        (else
                          (Class|FortPlayerPawn|SwapPawnCosmeticItemandVariant (Variables|Default|GetPlayerPawn) (Variables|Default|GetSwapGUID) (Utilities|GetDefaultOutfit "Random") false "None" false))))
                    (:Is%20Not%20Valid
                      ;[1 1 1] Swap to the last victim's loadout - LEGACY
                      (bind _output_get_3 (LocalVariables|SetEliminatedState (Class|FortPlayerControllerAthena|GetTrackedVictims _asfort_player_controller_athena)))
                      (Utilities|FlowControl|Sequence
                        (:then_0
                          ;[1 1 1] Fetch the loadout (happens before the direct above code)
                          (Utilities|IsValid (PlayerState|GetCurrentPawn _output_get_3)
                            (:Is%20Valid
                              (Utilities|CollectLoadoutActiveSwapswithActor (PlayerState|GetCurrentPawn _output_get_3)))
                            (:Is%20Not%20Valid
                              (bind _asfort_player_controller (Utilities|Casting|CastToFortPlayerController (Actor|GetOwner _output_get_3))
                                (:then
                                  (Utilities|CollectLoadoutActiveSwapswithController _asfort_player_controller))
                                (:CastFailed
                                  (bind _asfort_athena_aibot_controller (Utilities|Casting|CastToFortAthenaAIBotController (Actor|GetOwner _output_get_3))
                                    (:then
                                      (Utilities|CollectLoadoutActiveSwapswithController _asfort_athena_aibot_controller))
                                    (:CastFailed
                                      (Development|PrintString "No Valid Loadout Source"))))))))
                        (:then_1
                          (if (> (Utilities|Array|Length _activeswaps) 0)
                            (Class|FortPlayerPawn|SwapPawnCosmeticItemandVariant _playerpawn_2 _swapguid _activeswaps false "None" false)
                            (else
                              (Class|FortPlayerPawn|SwapPawnCosmeticItemandVariant _playerpawn_2 _swapguid (Utilities|GetDefaultOutfit "Random") false "None" false)))))))
                  (if (not (LocalVariables|GetbWasSwapActive))
                    ;[1 1 1] We want to monitor the pawn for things we don't like and remove the swap
                    (Ability|Advanced|GrantandActivateAbilityonSelf (Ability|GetAbilitySystemComponent (Variables|Default|GetPlayerPawn)) "/BRCosmetics/Abilities/Emotes/GAB_HightowerMango_CleanupEmote.GAB_HightowerMango_CleanupEmote_C" self)))
                (:Is%20Not%20Valid
                  _)))
            (:CastFailed
              (Development|PrintString "Invalid Controller"))))))))

(fn StopMontageIfNeeded
  (if (Variables|Default|GetbAbilityStopped)
    (CallFunction|ForceStopMontage)))
