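; Emote gameplay ability that makes the player performing it leave their ad-hoc squad.
; On activation it runs the parent emote logic, then continues only with authority and a valid player pawn.
; It gets the pawn's Athena controller and finds the AdHocSquads mutator through the game mode's mutator list, then asks that mutator to remove the player as a manual leave.
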(event Ability|EventActivateAbility
  (bind _playerpawn (Variables|Default|GetPlayerPawn))
  ;[0.1458 0.2535 1] Have the player performing the emote leave their ad-hoc squad
  (|CallParentFunction)
  (Ability|HasAuthority
    (:True
      (Utilities|IsValid _playerpawn
        (:Is%20Valid
          (bind _asfort_player_controller_athena (Utilities|Casting|CastToFortPlayerControllerAthena (Pawn|GetController _playerpawn)))
          (bind _asfort_game_mode_athena (Utilities|Casting|CastToFortGameModeAthena (Game|GetGameMode)))
          (bind _asfort_athena_mutator_ad_hoc_squads (Utilities|Casting|CastToFortAthenaMutator_AdHocSquads (Modifier|GetMutatorbyClass (Variables|/Script/FortniteGame.FortGameModeAthena|GetMutatorListComponent _asfort_game_mode_athena) "/Script/FortniteGame.FortAthenaMutator_AdHocSquads")))
          (Class|FortAthenaMutatorAdHocSquads|HavePlayerLeaveAdHocSquad _asfort_athena_mutator_ad_hoc_squads _asfort_player_controller_athena "ManualLeave"))))))
