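; Emote ability for a train-style group emote, where pawns join the leader in a chained line.
; It plays the montage, holstering the weapon if required, and keeps the leader open to interaction.
; Interacting with leader or follower starts the follower emote unless blocked by tags; joiners get a sync index, animation offset and tail target on the pawn ahead.
; When a follower leaves, those behind are re-indexed and relinked, without recomputing animation offsets.
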(event Ability|EventActivateAbility
  ;[0.01357 1 0] Activate Ability
  (|CallParentFunction)
  (bind _playerpawn (Variables|Default|GetPlayerPawn)
    (:then
      (Variables|/Script/FortniteGame.FortPlayerPawn|SetbIsWaitingForEmoteInteraction true _playerpawn) ; GAB_Emote_Generic::ActivateAbility sets IsWaitingForEmoteInteraction to false.\nWe need to make sure it gets set to true even if we do NOT play the montage (for example: if we already started the montage and the emote was requested again)
    )
    (:else)))

(event EventPlayInitialEmoteMontage
  (bind _montageinfo (Variables|Animation|GetMontageInfo))
  (bind _montagetoplay (Utilities|Struct|BreakFortGameplayAbilityMontageInfo _montageinfo))
  (bind _returnvalue (CallFunction|GetCurrentEmoteItemDefinition))
  (bind _playerpawn (Variables|Default|GetPlayerPawn))
  (bind _playerpawn_2 (Variables|Default|GetPlayerPawn))
  ;[1 1 1] Activate the montage
  (Utilities|IsValid _returnvalue
    (:Is%20Valid
      (if (Emote|ShouldHolsterWeapon _returnvalue _playerpawn)
        (Variables|/Script/FortniteGame.FortPlayerPawn|SetbIsWaitingForEmoteInteraction true _playerpawn)
        (Weapon|SetHolsterWeapon (Variables|Default|GetPlayerPawn) 0 true false (Variables|Default|GetEmoteHolsterId))
        (else
          _))
      (if (not (Variables|Default|GetbAbilityStopped))
        (Ability|Tasks|PlayMontageInfoAndWaitForTrigger "None" _montageinfo
          (:then
            (if (Variables|Default|GetbAbilityStopped)
              (CallFunction|ForceStopMontage)
              (else
                (Utilities|FlowControl|Sequence
                  (:then_0
                    ;[1 1 1] Show/hide reticle if requested
                    (if (Variables|Reticle|GetHideReticle)
                      (CallFunction|SetReticleVisibility true)))
                  (:then_1
                    ;[1 1 1] Bind Events
                    (Game|Player|BindEventtoOnEmoteInteracted _playerpawn_2 (delegate Custom|OnEmoteInteracted))
                    (EventDispatchers|BindEventtoOnGroupEmoteFollowerJoined _playerpawn_2 (delegate Custom|OnEmoteFollowerJoined))
                    (EventDispatchers|BindEventtoOnGroupEmoteFollowerLeft _playerpawn_2 (delegate Custom|OnEmoteFollowerLeft)))))))
          (:Cancelled
            (Utilities|FlowControl|Sequence
              (:then_0
                (Weapon|SetHolsterWeapon (Variables|Default|GetPlayerPawn) 0 false false (Variables|Default|GetEmoteHolsterId)))
              (:then_1
                (Utilities|IsValid _montagetoplay
                  (:Is%20Valid
                    (Utilities|FlowControl|Delay (Montage|GetDefaultBlendOutTime _montagetoplay)))
                  (:Is%20Not%20Valid
                    _))
                (Ability|EndAbility)))))))
    (:Is%20Not%20Valid
      _)))

(event Ability|EventOnEndAbility
  (params (bWasCancelled :type Boolean))
  (|CallParentFunction bWasCancelled)
  (Game|Player|UnbindEventfromOnEmoteInteracted (Variables|Default|GetPlayerPawn) (delegate Custom|OnEmoteInteracted)))

(event Custom|OnEmoteInteracted
  (params (Instigator :type Fort%20Player%20Pawn%20Object%20Reference))
  ;[1 1 1] OnEmoteInteracted
  (bind _asathena_dance_item_definition (Utilities|Casting|CastToAthenaDanceItemDefinition (CallFunction|GetCurrentEmoteItemDefinition)))
  (if (not (GameplayTags|HasAnyMatchingGameplayTags (Ability|GetAbilitySystemComponent Instigator) (Variables|Default|GetInstigatorTagsThatBlockGroupEmote)))
    (Variables|/Script/FortniteGame.FortPlayerPawn|SetGroupEmoteLookTarget (Variables|Default|GetPlayerPawn) Instigator)
    (Class|FortPlayerPawn|PlayGroupEmote Instigator (Variables|/Script/FortniteGame.AthenaDanceItemDefinition|GetGroupEmoteToStartFollower _asathena_dance_item_definition))))

(event Custom|OnEmoteFollowerJoined
  (params (JoiningPawn :type Fort%20Player%20Pawn%20Object%20Reference))
  (bind _groupemotefollowers (Variables|/Script/FortniteGame.FortPlayerPawn|GetGroupEmoteFollowers (Variables|Default|GetPlayerPawn)))
  (bind _returnvalue (Utilities|Array|Length _groupemotefollowers))
  (bind _playerpawn (Variables|Default|GetPlayerPawn))
  ;[1 1 1] OnEmoteFollowerJoined
  (bind _asathena_dance_item_definition (Utilities|Casting|CastToAthenaDanceItemDefinition (CallFunction|GetCurrentEmoteItemDefinition)))
  (bind _output_get (Variables|/Script/FortniteGame.FortPlayerPawn|SetGroupEmoteSyncValue (Math|Conversions|ToByte(Integer) (Utilities|Array|Length (Variables|/Script/FortniteGame.FortPlayerPawn|GetGroupEmoteFollowers _playerpawn))) JoiningPawn))
  (Variables|/Script/FortniteGame.FortPlayerPawn|SetGroupEmoteAnimOffset (* (Variables|/Script/FortniteGame.AthenaDanceItemDefinition|GetGroupSyncAnimOffset _asathena_dance_item_definition) (Math|Conversions|ToFloat(Byte) (Math|Byte|Byte+Byte _output_get 1))) JoiningPawn)
  (Variables|/Script/FortniteGame.FortPlayerPawn|SetbShouldSyncAnimationWithEmoteLeader true JoiningPawn)
  (Variables|/Script/FortniteGame.FortPlayerPawn|SetbIsWaitingForEmoteInteraction true JoiningPawn)
  (Game|Player|BindEventtoOnEmoteInteracted JoiningPawn (delegate Custom|FollowerInteracted))
  (Utilities|Array|Add _groupemotefollowers JoiningPawn)
  (if (> _returnvalue 1)
    (Variables|/Script/FortniteGame.FortPlayerPawn|SetGroupEmoteTailTarget (Utilities|Array|Get(acopy) _groupemotefollowers (- (Utilities|Array|LastIndex _groupemotefollowers) 1)) JoiningPawn)
    (elif (== _returnvalue 1)
      (Variables|/Script/FortniteGame.FortPlayerPawn|SetGroupEmoteTailTarget _playerpawn JoiningPawn))))

(event Custom|FollowerInteracted
  (params (Instigator :type Fort%20Player%20Pawn%20Object%20Reference))
  ;[1 1 1] FollowerInteracted
  (CallFunction|OnEmoteInteracted Instigator))

(event Custom|OnEmoteFollowerLeft
  (params (LeavingPawn :type Fort%20Player%20Pawn%20Object%20Reference))
  (bind _returnvalue (Utilities|Array|FindItem (Variables|/Script/FortniteGame.FortPlayerPawn|GetGroupEmoteFollowers (Variables|Default|GetPlayerPawn)) LeavingPawn))
  (bind _playerpawn (Variables|Default|GetPlayerPawn))
  (bind _groupemotefollowers (Variables|/Script/FortniteGame.FortPlayerPawn|GetGroupEmoteFollowers _playerpawn))
  ;[1 1 1] OnEmoteFollowerLeft
  (if (!= _returnvalue -1)
    (Variables|/Script/FortniteGame.FortPlayerPawn|SetGroupEmoteTailTarget 0 LeavingPawn)
    (Utilities|Array|RemoveIndex _groupemotefollowers _returnvalue)
    (for _index (range _returnvalue (+ (Utilities|Array|LastIndex _groupemotefollowers) 1))
      (Utilities|IsValid (Utilities|Array|Get(acopy) _groupemotefollowers _index)
        (:Is%20Valid
          (Variables|/Script/FortniteGame.FortPlayerPawn|SetGroupEmoteSyncValue (Math|Conversions|ToByte(Integer) _index) (Utilities|Array|Get(acopy) _groupemotefollowers _index))
          (if (== _index 0)
            (Variables|/Script/FortniteGame.FortPlayerPawn|SetGroupEmoteTailTarget _playerpawn (Utilities|Array|Get(acopy) _groupemotefollowers _index))
            (else
              (Utilities|IsValid (Utilities|Array|Get(acopy) _groupemotefollowers (- _index 1))
                (:Is%20Valid
                  (Variables|/Script/FortniteGame.FortPlayerPawn|SetGroupEmoteTailTarget (Utilities|Array|Get(acopy) _groupemotefollowers (- _index 1)) (Utilities|Array|Get(acopy) _groupemotefollowers _index)))))))))))
