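; Server-authoritative group emote ability that lets players form ad-hoc squads through a join-squad emote, driven by FortAthenaMutator_AdHocSquads.
; Activation checks team initialization and squad space, registers the invite, and reports the end reason on cleanup.
; Joining the emote squads both players before syncing.
; An optional looping timer line-traces from the camera to auto-squad players doing the same emote who face each other, selecting special dances.
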(event Ability|EventActivateAbility
  (bind _autosquadupfrequency (Variables|Default|GetAutoSquadUpFrequency))
  (bind _returnvalue (ScalableFloat|GetValueatLevel _autosquadupfrequency))
  ;[0.0142 1 0] EventActivateAbility - Notify FortAthenaMutator_AdHocSquads that a player has initiated a squad invite
  (|CallParentFunction)
  (Ability|HasAuthority
    (:True
      (bind _asfort_athena_mutator_ad_hoc_squads (Utilities|GetCachedAdHocSquadsMutator)
        (:then
          (bind _returnvalue_2 (Class|FortAthenaMutatorAdHocSquads|PlayerInitiatedSquadInvite _asfort_athena_mutator_ad_hoc_squads (Variables|Default|GetPlayerPawn)))
          (if _returnvalue_2
            ;;[0.1571 0.2974 1] Start a looping timer for auto squading up
            (if (> _returnvalue 0.0)
              (bind _returnvalue_3 (Utilities|Time|SetTimerbyEvent (EventDispatchers|CreateEvent :!event Custom|UpdateAutoSquadUp) _returnvalue true))
              (Variables|Default|SetUpdateAutoSquadTimerHandle _returnvalue_3))
            (else
              _)))
        (:Failed
          _))
      ;;[0.9219 0.5154 0.08197] If the invite fails, end the ability
      (Variables|Default|SetInvite%20Ended%20Reason "InviteCancelled_SquadFull") ; The only reason player squad invites fail at this point is because the squad is full.
      (Ability|EndAbility))))

(event Ability|EventOnEndAbility
  (params (bWasCancelled :type Boolean))
  ;[1 0.5711 0] OnEndAbility: Cleanup
  (|CallParentFunction bWasCancelled)
  (Ability|HasAuthority
    (:True
      (Utilities|Time|ClearandInvalidateTimerbyHandle (Variables|Default|GetUpdateAutoSquadTimerHandle))
      (bind _playerpawn (Variables|Default|GetPlayerPawn)
        (:then
          (bind _asfort_player_state_athena (Utilities|Casting|CastToFortPlayerStateAthena (Variables|/Script/Engine.Pawn|GetPlayerState _playerpawn)))
          (bind _asfort_athena_mutator_ad_hoc_squads (Utilities|GetCachedAdHocSquadsMutator))
          (Class|FortAthenaMutatorAdHocSquads|PlayerSquadInviteEnded _asfort_athena_mutator_ad_hoc_squads _asfort_player_state_athena (Variables|Default|GetInvite%20Ended%20Reason)))
        (:else)))))

(event EventStartGroupSyncEmote
  (params (Instigator%20Pawn :type Fort%20Player%20Pawn%20Object%20Reference))
  (bind _playerpawn (Variables|Default|GetPlayerPawn))
  ;[0.07239 0.1488 1] Start Group Sync Emote: Add both players participating in synced emote to ad-hoc team.
  (Ability|HasAuthority
    (:True
      (Utilities|IsValid Instigator%20Pawn
        (:Is%20Valid
          (bind _asfort_player_controller_athena (Utilities|Casting|CastToFortPlayerControllerAthena (Pawn|GetController Instigator%20Pawn)))
          (Utilities|IsValid _playerpawn
            (:Is%20Valid
              (bind _asfort_player_controller_athena_2 (Utilities|Casting|CastToFortPlayerControllerAthena (Pawn|GetController _playerpawn)))
              (bind _asfort_athena_mutator_ad_hoc_squads (Utilities|GetCachedAdHocSquadsMutator))
              (bind (_outsquadupresult _returnvalue) (Class|FortAthenaMutatorAdHocSquads|SquadUpPlayers _asfort_athena_mutator_ad_hoc_squads _asfort_player_controller_athena_2 _asfort_player_controller_athena)) ; For squading up the instigator is the player who started the emote (i.e. the Player Pawn)\nand the follower is the player who joined in on the emote (i.e. the Instigator Pawn)
              ;;[0.1639 1 0.572] If the squad-up was successful, THEN and ONLY THEN allow the synced emote to continue.
              (if _returnvalue
                (Variables|Default|SetInvite%20Ended%20Reason "InviteEnded_APlayerAcceptedTheInvite")
                (|CallParentFunction Instigator%20Pawn)))))))))

(event Custom|UpdateAutoSquadUp
  ;[0.4092 0.9101 1] UpdateAutoSquadUp: If there are any nearby players also doing the join squad emote, maybe squad up with them.
  (CallFunction|AttemptAutoSquadUp))

(fn AttemptAutoSquadUp
  (bind _localplayercameramanager (LocalVariables|GetLocalPlayerCameraManager))
  (bind _returnvalue (Camera|GetCameraRotation _localplayercameramanager))
  (bind _returnvalue_2 (Math|Vector|GetForwardVector _returnvalue))
  (bind _targetplayerpawn (LocalVariables|GetTargetPlayerPawn))
  (bind _returnvalue_3 (Ability|GetFortAbilitySystemComponent _targetplayerpawn))
  (bind _returnvalue_4 (Camera|GetCameraLocation _localplayercameramanager))
  (bind _returnvalue_5 (Utilities|Enum|LiteralenumEAdHocSquads_InviteStatus "InviteEnded_APlayerAcceptedTheInvite"))
  (bind _targetplayerpawn_2 (LocalVariables|GetTargetPlayerPawn))
  (bind _targetplayercontroller (LocalVariables|GetTargetPlayerController))
  (bind _capsulecomponent (Variables|/Script/Engine.Character|GetCapsuleComponent _targetplayerpawn_2))
  (bind _targetplayerjoinadhocsquadability (LocalVariables|GetTargetPlayerJoinAdHocSquadAbility))
  (bind _playercameramanager (Variables|/Script/Engine.PlayerController|GetPlayerCameraManager _targetplayercontroller))
  (Ability|HasAuthority
    (:True
      ;[0.2977 1 0.3244] Validate Data and Store off local variables.
      (bind _playerpawn (Variables|Default|GetPlayerPawn)
        (:then
          (Utilities|IsValid (Variables|/Script/Engine.Character|GetCapsuleComponent _playerpawn)
            (:Is%20Valid
              (LocalVariables|SetLocalPlayerCapsuleComponent (Variables|/Script/Engine.Character|GetCapsuleComponent _playerpawn))
              (bind _asfort_player_controller_athena (Utilities|Casting|CastToFortPlayerControllerAthena (Fortnite|GetFortPlayerControllerfromActor _playerpawn)))
              (bind _output_get (LocalVariables|SetLocalPlayerController _asfort_player_controller_athena))
              (Utilities|IsValid (Variables|/Script/Engine.PlayerController|GetPlayerCameraManager _output_get)
                (:Is%20Valid
                  (LocalVariables|SetLocalPlayerCameraManager (Variables|/Script/Engine.PlayerController|GetPlayerCameraManager _output_get))
                  ;[0.5156 0.2378 0] Early out if the player can no longer join a squad
                  (bind _asfort_athena_mutator_ad_hoc_squads (Utilities|GetCachedAdHocSquadsMutator)
                    (:then
                      (if (not (Class|FortAthenaMutatorAdHocSquads|IsPlayersSquadFull _asfort_athena_mutator_ad_hoc_squads (LocalVariables|GetLocalPlayerController)))
                        ;[0.3583 0.3382 1] We use a raycast to find if the local player is looking at another player pawn that is within Auto Squad Up range.
                        ; Create a ray that lies along the local camera's forward vector.
                        ; The ray's start location will be near the local player's capsule,
                        ; the ray's end location will be Auto Squad Up Distance away from the starting location, down the camera's forward vector
                        ;;[0.9765 1 0.1238] We want our raycast to start near the local player pawn, so we need to figure out the offset from the camera to location on the camer'a forward vector that is nearest the player pawn.
                        (LocalVariables|SetRaycastForwardOffset (* _returnvalue_2 (Math|Vector|DotProduct (- (Transformation|GetWorldLocation (Variables|/Script/Engine.Character|GetCapsuleComponent (Variables|Default|GetPlayerPawn))) _returnvalue_4) _returnvalue_2)))
                        ;[0.1491 0.2843 1] Perform the raycast and ONLY continue if the first object we hit is a FortPlayerPawn.
                        (bind _outhit (Collision|LineTraceByChannel (+ _returnvalue_4 (LocalVariables|GetRaycastForwardOffset)) (+ (LocalVariables|GetRaycastForwardOffset) (+ _returnvalue_4 (* _returnvalue_2 (ScalableFloat|GetScalableFloatMultiplier (Variables|Default|GetAutoSquadUpDistance))))) "TraceTypeQuery5" false (Utilities|Array|MakeArray (Variables|Default|GetPlayerPawn))))
                        (if (Collision|BreakHitResult _outhit)
                          (bind _asfort_player_pawn (Utilities|Casting|CastToFortPlayerPawn (.HitActor (Collision|BreakHitResult _outhit))))
                          (LocalVariables|SetTargetPlayerPawn _asfort_player_pawn)
                          (bind _asfort_player_controller_athena_2 (Utilities|Casting|CastToFortPlayerControllerAthena (Fortnite|GetFortPlayerControllerfromActor _asfort_player_pawn)))
                          (LocalVariables|SetTargetPlayerController _asfort_player_controller_athena_2)
                          ;[0.1102 0.5971 1] Continue ONLY if the Target pawn is also performing this emote.
                          (Utilities|IsValid _returnvalue_3
                            (:Is%20Valid
                              (bind _returnvalue_6 (Ability|GetPrimaryAbilityInstanceOfAbilitySystemComponent _returnvalue_3 (Utilities|GetClass self)))
                              (bind _asgab_group_emote_request_join_ad_hoc_squad (Utilities|Casting|CastToGAB_GroupEmoteRequest_JoinAdHocSquad _returnvalue_6))
                              (LocalVariables|SetTargetPlayerJoinAdHocSquadAbility _asgab_group_emote_request_join_ad_hoc_squad)
                              ;[0.811 0.4623 1] Make sure that the target player is looking towards the local player. This way the squad up won't be a surprise to the target player.
                              (Utilities|IsValid _playercameramanager
                                (:Is%20Valid
                                  (Utilities|IsValid _capsulecomponent
                                    (:Is%20Valid
                                      (if (< (Math|Trig|Acos(Degrees) (Math|Vector|DotProduct (Math|Vector|GetForwardVector (Camera|GetCameraRotation _playercameramanager)) (Math|Vector|Normalize (- (Transformation|GetWorldLocation (LocalVariables|GetLocalPlayerCapsuleComponent)) (Transformation|GetWorldLocation _capsulecomponent))))) (ScalableFloat|GetValueatLevel (Variables|Default|GetAutoSquadUpOtherPlayerAngleDegrees)))
                                        ;[0 0.1086 1] All pre-conditions passed, attempt to squad up.
                                        (bind _asfort_athena_mutator_ad_hoc_squads_2 (Utilities|GetCachedAdHocSquadsMutator))
                                        (bind (_outsquadupresult _returnvalue_7) (Class|FortAthenaMutatorAdHocSquads|SquadUpPlayers _asfort_athena_mutator_ad_hoc_squads_2 (LocalVariables|GetLocalPlayerController) (LocalVariables|GetTargetPlayerController)))
                                        (if _returnvalue_7
                                          ;;[1 0.4349 0] The Squad-up is successful!!! End the Join Squad emote for both players
                                          (Variables|Default|SetInvite%20Ended%20Reason _returnvalue_5 _targetplayerjoinadhocsquadability)
                                          (Variables|Default|SetPlayersSunkThroughAutoJoinUp true _targetplayerjoinadhocsquadability)
                                          (Variables|Default|SetInvite%20Ended%20Reason _returnvalue_5)
                                          (Variables|Default|SetPlayersSunkThroughAutoJoinUp true)
                                          (|CallParentFunction (LocalVariables|GetTargetPlayerPawn)) ; We call Super::SyncGroupEmote because this class's version of SyncGroupEmote calls SquadUpPlayers again.\nCalling that SquadUpPlayers again will cause it to fail (because the players will already be squaded up).\nThat will cause the ability to hang instead of completing with an auto-squad-up
                                        )))))))))
                        (else
                          (Variables|Default|SetInvite%20Ended%20Reason "InviteCancelled_SquadFull"))))
                    (:Failed
                      _))
                  ;;[1 0.4635 0.4635] Something is preventing this player from joining up with any more players, stop the emote.
                  (Ability|EndAbility))))))
        (:else)))))

(fn GetGroupEmote
  (params (SourceDanceItemDefinition :type Athena%20Dance%20Item%20Definition%20Object%20Reference) (OwnsGroupEmote :type Boolean) (IsLeader :type Boolean))
  ;[0.5477 0 1] If these players are syning their animation because of an Auto Squad Up
  ; then use special dances
  (bind _asathena_ad_hoc_squads_join_dance_item_definition (Utilities|Casting|CastToAthenaAdHocSquadsJoinDanceItemDefinition SourceDanceItemDefinition)
    (:then
      (bind _returnvalue (AdHocSquadUp|GetGroupEmoteforAdHocSquad _asathena_ad_hoc_squads_join_dance_item_definition (Variables|Default|GetPlayersSunkThroughAutoJoinUp) (Variables|GetIsLeader) (Variables|GetOwnsGroupEmote)))
      (Utilities|IsValid _returnvalue
        (:Is%20Valid
          (return _returnvalue))
        (:Is%20Not%20Valid
          _)))
    (:CastFailed
      _))
  (bind _returnvalue_2 (|CallParentFunction (Variables|GetSourceDanceItemDefinition) (Variables|GetOwnsGroupEmote) (Variables|GetIsLeader)))
  (return _returnvalue_2))

(fn CanActivateAbility
  (params (ActorInfo :type Gameplay%20Ability%20Actor%20Info%20Structure) (Handle :type Gameplay%20Ability%20Spec%20Handle%20Structure))
  (bind (_owneractor _avataractor _playercontroller) (Utilities|Struct|BreakGameplayAbilityActorInfo ActorInfo))
  (Ability|HasAuthority
    (:True
      (bind _asfort_game_mode_athena (Utilities|Casting|CastToFortGameModeAthena (Ability|GetGameModefromActorInfo ActorInfo)))
      (bind _asfort_athena_mutator_ad_hoc_squads (Utilities|Casting|CastToFortAthenaMutator_AdHocSquads (Modifier|GetMutatorbyClass (Variables|/Script/FortniteGame.FortGameModeAthena|GetMutatorListComponent _asfort_game_mode_athena) "/Script/FortniteGame.FortAthenaMutator_AdHocSquads")))
      (if (or (Class|FortAthenaMutatorAdHocSquads|ShouldAllowRequestsBeforePlayerInitialAssignments _asfort_athena_mutator_ad_hoc_squads) (Class|FortAthenaMutatorAdHocSquads|GetTeamsHaveBeenInitialized _asfort_athena_mutator_ad_hoc_squads))
        (bind _asfort_player_controller_athena (Utilities|Casting|CastToFortPlayerControllerAthena _playercontroller))
        (if (Class|FortAthenaMutatorAdHocSquads|IsPlayersSquadFull _asfort_athena_mutator_ad_hoc_squads _asfort_player_controller_athena)
          (return "(GameplayTags=((TagName=\"AdHoc.FailureReason.Join.FullSquad\")))" false))
        (else
          (return "(GameplayTags=((TagName=\"AdHoc.FailureReason.Generic\")))" false))))
    (:False
      _))
  (bind (_relevanttags _returnvalue) (|CallParentFunction ActorInfo Handle))
  (return _relevanttags _returnvalue))
