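; Gameplay ability that maintains, then cleans up, the Hightower Mango emote's cosmetic character swap.
; It ends through a gameplay cue on timeout, excess distance, DBNO, carrying, locker entry, death, aircraft entry or qualifying damage.
; A looping timer checks distance from a cached location, re-cached on vehicle changes.
; Authority-only unswap deactivates the data store's last victim loadout, else a default, immediately on death or aircraft entry to avoid respawn glitches.
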(event Ability|EventActivateAbility
  (bind _datatabletimeout (Variables|Default|GetDatatableTimeout))
  (bind _returnvalue (ScalableFloat|GetValueatLevel _datatabletimeout))
  (bind _outputdelegate (EventDispatchers|CreateEvent :!event CachePlayerComparisonLocation))
  (bind _controller (Variables|Default|GetController))
  (bind _playerpawn (Variables|Default|GetPlayerPawn))
  (|CallParentFunction)
  (CallFunction|SetupTagBasedCancels)
  (CallFunction|SetupEventBasedCancels)
  (bind _asfort_player_pawn (Utilities|Casting|CastToFortPlayerPawn (Ability|GetAvatarActorfromActorInfo)))
  (bind _output_get (Variables|Default|SetPlayerPawn _asfort_player_pawn))
  (Variables|Default|SetController (Fortnite|GetFortPlayerControllerfromActor _output_get))
  (Utilities|FlowControl|Sequence
    (:then_0
      ;[1 1 1] Allow player to get in/out of vehicles without losing the ability
      (Variables|Default|GetController
        (:then
          (Vehicle|AssignOnEnterVehicleDriver _controller _outputdelegate)
          (Vehicle|AssignOnEnterVehiclePassenger _controller _outputdelegate)
          (Vehicle|AssignOnFullyExitVehicle _controller _outputdelegate))
        (:else)))
    (:then_1
      (CallFunction|CachePlayerComparisonLocation)
      (Utilities|Time|SetTimerbyEvent (delegate Custom|DistanceCheck) (Variables|Default|GetDistanceCheckFrequency) true)
      (if (> _returnvalue 0.0)
        (Utilities|FlowControl|Delay _returnvalue)
        (CallFunction|ApplyCueThenExit)))
    (:then_2
      (EventDispatchers|BindEventtoOnDied _playerpawn (delegate Custom|OnDied))
      (EventDispatchers|AssignOnEnteredAircraft _playerpawn (delegate Custom|OnEnteredAircraft)))))

(event Custom|DistanceCheck
  (bind _playerpawn (Variables|Default|GetPlayerPawn)
    (:then
      (if (> (Math|Vector|DistanceSquared(Vector) (CallFunction|GetPlayerComparisonLocation _playerpawn) (Variables|Default|GetInitialLocation)) (Math|Float|Square (ScalableFloat|GetValueatLevel (Variables|Default|GetDatatableMaxDistanceBeforeStop))))
        (CallFunction|ApplyCueThenExit)))
    (:else)))

(event Custom|OnDied
  (params (DamagedActor :type Actor%20Object%20Reference) (Damage :type Float%20%28single-precision%29) (InstigatedBy :type Controller%20Object%20Reference) (DamageCauser :type Actor%20Object%20Reference) (HitLocation :type Vector) (FHitComponent :type Primitive%20Component%20Object%20Reference) (BoneName :type Name) (Momentum :type Vector))
  ;[1 1 1] Have to unswap immediately, if the controller is gone already, it will glitch on respawn
  (CallFunction|UnswapCharacterParts)
  (CallFunction|ApplyCueThenExit))

(event Custom|SetupTagBasedCancels
  (Utilities|FlowControl|Sequence
    (:then_0
      (Ability|Tasks|WaitGameplayTagAdd "(TagName=\"Gameplay.Status.DBNO\")"
        (:Added
          _)))
    (:then_1
      (Ability|Tasks|WaitGameplayTagAdd "(TagName=\"Gameplay.Action.Player.CarryingPlayer\")"
        (:Added
          _)))
    (:then_2
      (Ability|Tasks|WaitGameplayTagAdd "(TagName=\"Granted.Athena.Locker.Inside\")"
        (:Added
          _)) ; Cancel the effect if the player enters a locker to change skins
    ))
  (CallFunction|ApplyCueThenExit))

(event Custom|OnEnteredAircraft
  ;[1 1 1] Have to unswap immediately, if the controller is gone already, it will glitch on respawn
  (CallFunction|UnswapCharacterParts)
  (CallFunction|ApplyCueThenExit))

(event Custom|SetupEventBasedCancels
  (Utilities|FlowControl|Sequence
    (:then_0
      (bind (_asynctaskproxy _payload) (Ability|Tasks|WaitGameplayEvent "(TagName=\"Event.Damage.Dealt\")" 0 false false)
        (:EventReceived
          (if (!= (.EventMagnitude (Utilities|Struct|BreakGameplayEventData _payload)) 0.0)
            (switch Utilities|FlowControl|Switch|SwitchonGameplayTag (Utilities|Struct|BreakGameplayEventData _payload)
              (:Event.Damage.Dealt
                ;[1 1 1] We don't want it to be just _anything_ that you damage to get out of this
                (Utilities|Casting|CastToFortPlayerPawn (.Target (Utilities|Struct|BreakGameplayEventData _payload))
                  (:then
                    _)
                  (:CastFailed
                    (Utilities|Casting|CastToFortAthenaVehicle (.Target (Utilities|Struct|BreakGameplayEventData _payload))))))
              (:Event.Damage.Dealt.Shield
                _))
            (CallFunction|ApplyCueThenExit)))))
    (:then_1
      (bind (_asynctaskproxy_2 _payload_2) (Ability|Tasks|WaitGameplayEvent "(TagName=\"Event.Damage.Received\")" 0 false false)
        (:EventReceived
          (if (!= (.EventMagnitude (Utilities|Struct|BreakGameplayEventData _payload_2)) 0.0)
            _))))))

(event Ability|EventOnEndAbility
  (params (bWasCancelled :type Boolean))
  (|CallParentFunction bWasCancelled)
  (CallFunction|UnswapCharacterParts))

(event Custom|ApplyCueThenExit
  (Ability|ExecuteGameplayCueOnOwner (Variables|Default|GetGC) (Utilities|Struct|MakeGameplayEffectContextHandle))
  (Networking|ForceNetUpdate (Ability|GetAvatarActorfromActorInfo))
  (Utilities|FlowControl|Delay 0.1)
  (Ability|EndAbility))

(fn CachePlayerComparisonLocation
  (Variables|Default|SetInitialLocation (Variables|/Script/Engine.SceneComponent|GetRelativeLocation (Variables|/Script/Engine.Actor|GetRootComponent (Variables|Default|GetPlayerPawn)))))

(fn GetPlayerComparisonLocation
  (params (Pawn :type Actor%20Object%20Reference))
  (return (Variables|/Script/Engine.SceneComponent|GetRelativeLocation (Variables|/Script/Engine.Actor|GetRootComponent Pawn))))

(fn UnswapCharacterParts
  (bind _swapguid (Variables|Default|GetSwapGUID))
  (bind _playerpawn (Variables|Default|GetPlayerPawn))
  (Ability|HasAuthority
    (:True
      (bind _asfort_player_controller_athena (Utilities|Casting|CastToFortPlayerControllerAthena (Pawn|GetController (Variables|Default|GetPlayerPawn)))
        (:then
          (bind _asfort_data_store_hightower_mango_data_store (Utilities|Casting|CastToFortDataStore_HightowerMangoDataStore (DataStore|GetNamedStorebyType (Actor|GetComponentbyClass _asfort_player_controller_athena "/Script/FortniteGame.FortGenericDataStoreManagerComponent") "MangoDataStore" "/Script/FortniteGame.FortDataStore_HightowerMangoDataStore"))
            (:then
              (if (> (Utilities|Array|Length (Variables|/Script/FortniteGame.FortDataStore_HightowerMangoDataStore|GetLastVictimRelevantLoadout _asfort_data_store_hightower_mango_data_store)) 0)
                (Class|FortPlayerPawn|SwapPawnCosmeticItemandVariant _playerpawn _swapguid (Variables|/Script/FortniteGame.FortDataStore_HightowerMangoDataStore|GetLastVictimRelevantLoadout _asfort_data_store_hightower_mango_data_store) false "Inactive" false)
                (else
                  _)))
            (:CastFailed
              _)) ; this cast seems strange until you realize it won't compile without it
        )
        (:CastFailed
          (Development|PrintString "No player controller")))
      (Class|FortPlayerPawn|SwapPawnCosmeticItemandVariant _playerpawn _swapguid (Variables|Default|GetSwapToRamirez) false "Inactive" false))))
